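; Parent Blueprint for Mantis weapons, under the Mantis item folder.
; On BeginPlay it casts its instigator to FortPlayerPawn and stores it as the PlayerCharacter variable.
; It declares empty ReticleStatusDisplay and ShowReticleStatusDisplay events, apparently hooks for child weapons to drive reticle status UI.
; The construction script only calls the parent implementation.
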(event EventBeginPlay
  (bind _asfort_player_pawn (Utilities|Casting|CastToFortPlayerPawn (Game|GetInstigator)))
  (Variables|Runtime|SetPlayerCharacter _asfort_player_pawn))

(event Custom|ReticleStatusDisplay
  (params (Hide :type Boolean)))

(event Custom|ShowReticleStatusDisplay)

(fn ConstructionScript
  (|CallParentFunction))
