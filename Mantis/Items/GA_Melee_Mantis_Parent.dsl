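; Parent gameplay ability for Mantis melee attack sequences, handling aim-assisted motion warping, weapon durability and rarity-based effect levels.
; Activation derives effect level from weapon rarity and can end the ability on swimming.
; A targeting preset picks the aim-assist target, with time-gated switching, for root motion warping, which rejects dead, tagged or too-close pawns.
; On authority, hits queue unique technique names, charged as durability on commit events or ability end.
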(event Custom|GetAimAssistTarget
  (bind _aimassisttarget (Variables|Default|GetAimAssistTarget))
  (bind _targeting_preset (Variables|Default|GetTargeting%20Preset))
  (Utilities|IsValid _targeting_preset
    (:Is%20Valid
      (bind (_asynctaskproxy _targetinghandle) (Class|AsyncActionPerformTargeting|PerformTargetingAsyncAction (CallFunction|GetAvatarasFortPlayerPawn) _targeting_preset)
        (:Targeted
          (if (> (Utilities|Array|Length (TargetingSystem|TargetingResults|GetTargetingResultsActors "/Script/TargetingSystem.Default__TargetingSubsystem" _targetinghandle)) 0)
            (bind _returnvalue (CallFunction|IsValidTarget (Utilities|Array|Get(acopy) (TargetingSystem|TargetingResults|GetTargetingResultsActors "/Script/TargetingSystem.Default__TargetingSubsystem" _targetinghandle))))
            (if _returnvalue
              (Utilities|IsValid _aimassisttarget
                (:Is%20Valid
                  (if (Utilities|NotEqual(Object) _aimassisttarget (Utilities|Array|Get(acopy) (TargetingSystem|TargetingResults|GetTargetingResultsActors "/Script/TargetingSystem.Default__TargetingSubsystem" _targetinghandle)))
                    (bind _returnvalue_2 (CallFunction|ShouldSwitchTargets (TargetingSystem|TargetingResults|GetTargetingResultsActors "/Script/TargetingSystem.Default__TargetingSubsystem" _targetinghandle) (Variables|Default|GetAimAssistTarget)))
                    (if _returnvalue_2
                      _)
                    (else
                      _))
                  (Variables|Default|SetAimAssistTarget (Utilities|Array|Get(acopy) (TargetingSystem|TargetingResults|GetTargetingResultsActors "/Script/TargetingSystem.Default__TargetingSubsystem" _targetinghandle)))
                  (Variables|Default|SetHasAimAssistTarget true)
                  (Variables|Default|SetTimeOfLastTarget (Utilities|Time|GetGameTimeinSeconds)))
                (:Is%20Not%20Valid
                  _))
              (else
                _)))
          (Variables|Default|SetHasAimAssistTarget false)
          (Variables|Default|SetAimAssistTarget 0))))))

(event Ability|EventActivateAbility
  ;[0 0.15 0.000587 0.5] ActivateAbility: Invoked whenever a new mantis attack sequence is started.
  (|CallParentFunction)
  (Utilities|FlowControl|Sequence
    (:then_0
      (bind _output_get (CallFunction|InitAbility))
      (if (Variables|Default|GetbForceUnCrouch)
        (Character|UnCrouch _output_get)
        (else
          _))
      (Class|BMantisWeaponParent|ReticleStatusDisplay (Utilities|Casting|CastToB_Mantis_Weapon_Parent (CallFunction|GetBestValidWeapon)) false)
      (if (not (Variables|Default|GetCanSwingInWater))
        ;;[1 1 1] kill ability if we enter water
        (bind _asynctaskproxy (Ability|Tasks|WaitGameplayTagAdd "(TagName=\"MovementMode.SurfaceSwimming\")")
          (:then
            (Variables|Default|SetWaitStartSwimmingAsyncTask _asynctaskproxy))
          (:Added
            (Ability|EndAbility)))))
    (:then_1
      ;;[0.02353 0.07859 0.15 0.5] Track durability.
      (Ability|HasAuthority
        (:True
          (bind _busesdurability (Durability|DoesSourceWeaponUseDurability))
          (if _busesdurability
            (bind _asynctaskproxy_2 (Ability|Tasks|WaitGameplayEvent "(TagName=\"Event.Mantis.Durability.Commit\")")
              (:then
                (Variables|Transient|Durability|SetAsyncTask_WaitDurablityCommitEvent _asynctaskproxy_2))
              (:EventReceived
                (Durability|CommitDurabilityCost)))))))))

(event Mantis|EventOnMantisTechniqueHit
  (params (TargetDataHandle :type Gameplay%20Ability%20Target%20Data%20Handle%20Structure%20%28by%20ref%29) (ApplicationTag :type Gameplay%20Tag%20Structure))
  ;[0 0.15 0.1423 0.5] OnMantisTechniqueHit: Use the weapon durability when the weapon does damage on hit.
  (|CallParentFunction TargetDataHandle ApplicationTag)
  (Ability|HasAuthority
    (:True
      (bind _busesdurability (Durability|DoesSourceWeaponUseDurability))
      (if _busesdurability
        ;;[1 1 1] Techniques can hit multiple times; account for only the first damage.
        (Utilities|Array|AddUnique (Variables|Transient|Durability|GetPendingDurabilityUseTechniques) (Variables|Transient|Durability|GetCurrentTechniqueName))))))

(event Mantis|EventOnMantisAblilityTechniqueStarted
  (params (OwningTags :type Gameplay%20Tag%20Container%20Structure) (SequenceName :type Name :default "None") (StartsSequence :type Boolean) (EndsSequence :type Boolean))
  ;[0.03297 0.07217 0.15 0.5] MantisAbilityTechniqueStarted: Track the name of the current technique.
  (|CallParentFunction OwningTags SequenceName StartsSequence EndsSequence)
  (Variables|Transient|Durability|SetCurrentTechniqueName SequenceName))

(event Ability|EventOnEndAbility
  (params (bWasCancelled :type Boolean))
  ;[0.09896 0.06391 0 0.5] EndPlay: Clean up, and commit any pending durability cost.
  (Utilities|FlowControl|Sequence
    (:then_0
      (Variables|Transient|Durability|SetCurrentTechniqueName "None")
      (Ability|HasAuthority
        (:True
          (bind _asynctask_waitdurablitycommitevent (Variables|Transient|Durability|GetAsyncTask_WaitDurablityCommitEvent)
            (:then
              (GameplayTasks|EndTask _asynctask_waitdurablitycommitevent))
            (:else
              _))
          (Durability|CommitDurabilityCost))))
    (:then_1
      (Class|BMantisWeaponParent|ReticleStatusDisplay (Utilities|Casting|CastToB_Mantis_Weapon_Parent (CallFunction|GetBestValidWeapon)) false)
      (bind _waitstartswimmingasynctask (Variables|Default|GetWaitStartSwimmingAsyncTask)
        (:then
          (GameplayTasks|EndTask _waitstartswimmingasynctask))
        (:else)))
    (:then_2
      (|CallParentFunction bWasCancelled))))

(fn CommitDurabilityCost
  (bind _returnvalue (Ability|GetCurrentSourceWeapon))
  (bind _pendingdurabilityusetechniques (Variables|Transient|Durability|GetPendingDurabilityUseTechniques))
  ;[0 0.1288 0.15 0.5] CommitDurabilityCost: Based on hit count, take away durability points.
  (Ability|HasAuthority
    (:True
      (bind _busesdurability (Durability|DoesSourceWeaponUseDurability))
      (if _busesdurability
        (Utilities|IsValid _returnvalue
          (:Is%20Valid
            (FortWeapon|UseWeaponDurability(Scaled) _returnvalue (* (ScalableFloat|GetValueatLevel (Variables|Hotfix|GetHotfix_HitDurabilityCost)) (Utilities|Array|Length _pendingdurabilityusetechniques)))
            (Utilities|Array|Clear _pendingdurabilityusetechniques)))))))

(fn ConfigureRootMotionWarping
  (params (TechniqueName :type Name :default "None"))
  (bind (_allowtranslationwarp _allowrotationwarp _snapshottargetlocation) (CallFunction|GetRootMotionOptions))
  (bind _aimassisttarget (Variables|Default|GetAimAssistTarget))
  (CallFunction|GetAimAssistTarget)
  (if (Variables|Default|GetHasAimAssistTarget)
    (bind _returnvalue (CallFunction|IsValidTarget _aimassisttarget))
    (if _returnvalue
      (return _allowtranslationwarp _allowrotationwarp _aimassisttarget _snapshottargetlocation false "0, 0, 0" false 0)
      (else
        (return false false 0 false false "0, 0, 0" false 0)))
    (else
      (return false false 0 false false "0, 0, 0" false 0))))

(fn DoesSourceWeaponUseDurability
  ;[0.03907 0.1168 0.15 0.5] Returns if the source weapon uses durability.
  (Utilities|FlowControl|Sequence
    (:then_0
      (bind _returnvalue (CallFunction|GetBestValidWeapon))
      (Utilities|IsValid _returnvalue
        (:Is%20Valid
          (bind _returnvalue_2 (FortWeapon|GetInventoryItem _returnvalue))
          (Utilities|IsValid _returnvalue_2
            (:Is%20Valid
              (return (Item|HasDurability _returnvalue_2)))))))
    (:then_1
      (return false))))

(fn GetBestValidWeapon
  (bind _returnvalue (Ability|GetCurrentSourceWeapon))
  (Utilities|IsValid _returnvalue
    (:Is%20Valid
      (return _returnvalue))
    (:Is%20Not%20Valid
      (bind _outfortplayerpawn (CallFunction|GetAvatarasValidFortPlayerPawn)
        (:ValidFortPlayerPawn
          (return (Weapon|GetCurrentWeapon _outfortplayerpawn)))
        (:AvatarCastFailed
          (return 0))))))

(fn GetLevelForGameplayEffectContainer
  (params (ApplicationTag :type Gameplay%20Tag%20Structure))
  (return (Math|Float|Truncate (Variables|Default|GetGELevel))))

(fn GetRootMotionOptions
  (bind (_asaicontroller _bsuccess) (Utilities|Casting|CastToAIController (Pawn|GetController (Ability|GetActivatingPawn))))
  (if (Variables|Mantis|GetAI%20Override%20Mantis%20Option)
    (if _bsuccess
      (return (Variables|Mantis|AI|GetAI%20Allow%20Translation%20Warp) (Variables|Mantis|AI|GetAI%20Allow%20Rotation%20Warp) (Variables|Mantis|AI|GetAI%20Snapshot%20Target%20Location))
      (else
        _)))
  (return (Variables|Mantis|GetAllow%20Translation%20Warp) (Variables|Mantis|GetAllow%20Rotation%20Warp) (Variables|Mantis|GetSnapshot%20Target%20Location)))

(fn InitAbility
  (bind _asfort_player_pawn (Utilities|Casting|CastToFortPlayerPawn (Ability|GetActivatingPawn))
    (:then
      (bind _output_get (Variables|Default|SetPlayerCharacter _asfort_player_pawn))
      (if (Tags|WeaponDataHasMatchingGameplayTag (Variables|/Script/FortniteGame.FortPawn|GetCurrentWeapon _output_get) (Variables|Default|GetRarityUncommon))
        (Variables|Default|SetAdditionalLevels 1.0)
        (elif (Tags|WeaponDataHasMatchingGameplayTag (Variables|/Script/FortniteGame.FortPawn|GetCurrentWeapon _output_get) (Variables|Default|GetRarityRare))
          (Variables|Default|SetAdditionalLevels 2.0)
          (elif (Tags|WeaponDataHasMatchingGameplayTag (Variables|/Script/FortniteGame.FortPawn|GetCurrentWeapon _output_get) (Variables|Default|GetRarityEpic))
            (Variables|Default|SetAdditionalLevels 3.0)
            (elif (Tags|WeaponDataHasMatchingGameplayTag (Variables|/Script/FortniteGame.FortPawn|GetCurrentWeapon _output_get) (Variables|Default|GetRarityLegendary))
              (Variables|Default|SetAdditionalLevels 4.0)
              (elif (Tags|WeaponDataHasMatchingGameplayTag (Variables|/Script/FortniteGame.FortPawn|GetCurrentWeapon _output_get) (Variables|Default|GetRarityMythic))
                (Variables|Default|SetAdditionalLevels 5.0)
                (else
                  _))))))
      (Variables|Default|SetGELevel (+ (Variables|Default|GetDefaultLevel) (Variables|Default|GetAdditionalLevels))))
    (:CastFailed
      _))
  (return _output_get))

(fn IsValidTarget
  (params (Actor :type Actor%20Object%20Reference))
  (Utilities|IsValid Actor
    (:Is%20Valid
      (bind _asfort_pawn (Utilities|Casting|CastToFortPawn Actor)
        (:then
          (if (FortPawn|IsDead _asfort_pawn)
            (return false)
            (elif (GameplayTags|HasMatchingGameplayTag _asfort_pawn (Variables|Default|GetInvalidWarpTargetTag))
              (return false)
              (else
                ;[1 1 1] Prevent root motion warping towards a target that is too close; otherwise, it causes rapid rotation
                (return (> (Class|AbilityFunctions|GetDistanceBetweenActors (Ability|GetAvatarActorfromActorInfo) _asfort_pawn) (Variables|Default|GetMinDistanceToTarget_0)))))))
        (:CastFailed
          (return false))))
    (:Is%20Not%20Valid
      (return false))))

(fn SetKnockbackImmunity
  (params (ImmunityOn :type Boolean))
  (if ImmunityOn
    ;[1 1 1] Apply Knockback Immunity
    (bind _returnvalue (Ability|ApplyGameplayEffectSpecToOwner (Ability|MakeOutgoingGameplayEffectSpec (Variables|UtilityFunctionVars|GetGE_KnockbackImmunity))))
    (Variables|UtilityFunctionVars|SetHandle_KnockbackImmunity _returnvalue)
    (else
      ;[1 1 1] Remove Knockback Immunity
      (Ability|RemoveGameplayEffectFromOwnerWithHandle (Variables|UtilityFunctionVars|GetHandle_KnockbackImmunity)))))

(fn SetPawnCollision
  (params (Fort%20Pawn :type Fort%20Pawn%20Object%20Reference) (Collision%20On :type Boolean))
  (bind _savedcollisionchannel (Variables|UtilityFunctionVars|GetSavedCollisionChannel))
  (bind _capsulecomponent (Variables|/Script/Engine.Character|GetCapsuleComponent Fort%20Pawn))
  (Utilities|IsValid Fort%20Pawn
    (:Is%20Valid
      (if Collision%20On
        ;[1 1 1] Set Pawn collision back to blocking if the saved collision is not none
        (if (Utilities|Name|NotEqual(Name) _savedcollisionchannel "None")
          (Collision|SetCollisionProfileName _capsulecomponent _savedcollisionchannel true :!class /Script/Engine.PrimitiveComponent)
          (Collision|SetCollisionResponsetoChannel _capsulecomponent "ECC_Pawn" "ECR_Block")
          (Collision|SetCollisionResponsetoChannel _capsulecomponent "ECC_GameTraceChannel4" "ECR_Block"))
        (else
          ;[1 1 1] Set Pawn collision to overlap and save out collision
          (Variables|UtilityFunctionVars|SetSavedCollisionChannel (Collision|GetCollisionProfileName _capsulecomponent))
          (Collision|SetCollisionResponsetoChannel _capsulecomponent "ECC_Pawn" "ECR_Overlap")
          (Collision|SetCollisionResponsetoChannel _capsulecomponent "ECC_GameTraceChannel4" "ECR_Overlap"))))))

(fn ShouldSwitchTargets
  (params (AllTarget :type Array%20of%20Actor%20Object%20References) (CurrentTarget :type Actor%20Object%20Reference%20%28by%20ref%29))
  ;[1 1 1] If the current target wasn't in the list of actors targeted, it means it's no longer in range and we don't need to check any further
  (if (Utilities|Array|ContainsItem AllTarget CurrentTarget)
    (return (> (- (Utilities|Time|GetGameTimeinSeconds) (Variables|Default|GetTimeOfLastTarget)) (Variables|Default|GetMinTimeBetweenTargets)))
    (else
      (return true))))
